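# Conversation database for Nisha, the CeremonialGuardFencer NPC, defining the greeting she speaks when a player talks to her.
# A single entry point feeds one Speech node with a default line and contextual lines gated by first-ever conversation, per-NPC match conversation count, or the player's outfit.
# Priority curve rows rank first-meeting and outfit reactions.
# The Speech node then branches through link nodes to shared Quests, BasicServices and this NPC's own Dialogue entries.
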ConversationDatabase /NPCLibrary/NPCs/CeremonialGuardFencer/NPCLibrary_Conversation_CeremonialGuardFencer
entry_points [{tag=Conversation.Entry.CeremonialGuardFencer, destinations [#1]}]
nodes [
 ConversationEntryPointNode_7 {
  id=#1
  class=ConversationEntryPointNode
  outputs [#2]
  properties {entryTag {tagName=Conversation.Entry.CeremonialGuardFencer}}
 }
 Speech_C_4 {
  id=#2
  class=Speech_C
  outputs [#3, #4, #5]
  properties {
   nodeName=""
   generalConfig {
    defaultMessage=I do not miss The Society... but sometimes I miss knowing the hierarchy of things.
    contextualMessages [
     {
      message="I am Nisha, newly of The Underground. Much to my chagrin."
      contextRequirements [{requirement {class=FortControllerRequirement_IsFirstEverConversationWithNPC}}]
      priority {curve {rowName=Tandem.Speech.Priority.FirstEverInteraction}}
     }
     {
      message=I do not miss The Society... but sometimes I miss knowing the hierarchy of things.
      contextRequirements [{requirement {
         class=FortControllerRequirement_CheckNPCMatchConversationCount
         properties {nPCUniqueIdTag {tagName=AISpawnerData.Type.Tandem.CeremonialGuardFencer}}
        }}]
      priority {curve {rowName=None}}
     }
     {
      message=I thought the Island could use a strong hand. Doom has proven that analysis... flawed.
      contextRequirements [{requirement {
         class=FortControllerRequirement_CheckNPCMatchConversationCount
         properties {nPCUniqueIdTag {tagName=AISpawnerData.Type.Tandem.CeremonialGuardFencer}}
        }}]
      priority {curve {rowName=None}}
     }
     {
      message=This Island is swarming with superheroes. They can be a bit much.
      contextRequirements [{requirement {
         class=FortControllerRequirement_CheckNPCMatchConversationCount
         properties {nPCUniqueIdTag {tagName=AISpawnerData.Type.Tandem.CeremonialGuardFencer}}
        }}]
      priority {curve {rowName=None}}
     }
     {
      message=I guess this is the start of my rebellious phase. Hm.
      contextRequirements [{requirement {
         class=FortControllerRequirement_CheckNPCMatchConversationCount
         properties {nPCUniqueIdTag {tagName=AISpawnerData.Type.Tandem.CeremonialGuardFencer}}
        }}]
      priority {curve {rowName=None}}
     }
     {
      message="I will work with you, Hope. But that does not make us friends."
      contextRequirements [{requirement {
         class=FortControllerRequirement_HasCID
         properties {allowedCIds [
           &Character_VoidRedemption_Rebel
           &Character_BrightDisk
          ]}
        }}]
      priority {curve {rowName=Tandem.Speech.Priority.OutfitReaction}}
     }
     {
      message=Our parents would despise The Underground. That... almost makes this fun.
      contextRequirements [{requirement {class=FortControllerRequirement_HasCID, properties {allowedCIds [/BRCosmetics/Athena/Items/Cosmetics/Characters/Character_NovelGuard]}}}]
      priority {curve {rowName=Tandem.Speech.Priority.OutfitReaction}}
     }
     {
      message=I am sorry about your banana friend. For what that's worth.
      contextRequirements [{requirement {
         class=FortControllerRequirement_HasCID
         properties {allowedCIds [&Character_UndergroundRebel_Fashion]}
        }}]
      priority {curve {rowName=Tandem.Speech.Priority.OutfitReaction}}
     }
     {
      message="It feels odd to say, but you seem... freer, now that you're on your own."
      contextRequirements [{requirement {
         class=FortControllerRequirement_HasCID
         properties {allowedCIds [&Character_BengalBasher_Suit]}
        }}]
      priority {curve {rowName=Tandem.Speech.Priority.OutfitReaction}}
     }
     {
      message="I've thrown in with The Underground. If you cannot beat them, etc."
      contextRequirements [{requirement {
         class=FortControllerRequirement_HasCID
         properties {allowedCIds [/BRCosmetics/Athena/Items/Cosmetics/Characters/Character_DiamondHeart_Chic]}
        }}]
      priority {curve {rowName=Tandem.Speech.Priority.OutfitReaction}}
     }
     {
      message=I heard you helped dethrone a god. You seem like someone I should know.
      contextRequirements [{requirement {
         class=FortControllerRequirement_HasCID
         properties {allowedCIds [/BRCosmetics/Athena/Items/Cosmetics/Characters/Character_ReptilianOcean_Sleek]}
        }}]
      priority {curve {rowName=Tandem.Speech.Priority.OutfitReaction}}
     }
     {
      message=I'll put down my foil if you'll put down your bow. Allies?
      contextRequirements [{requirement {class=FortControllerRequirement_HasCID, properties {allowedCIds [/BRCosmetics/Athena/Items/Cosmetics/Characters/Character_SparkArcher]}}}]
      priority {curve {rowName=Tandem.Speech.Priority.OutfitReaction}}
     }
     {
      message=You're part of The Underground? That's... wonderful...
      contextRequirements [{requirement {class=FortControllerRequirement_HasCID, properties {allowedCIds [/BRCosmetics/Athena/Items/Cosmetics/Characters/Character_FeralTrash]}}}]
      priority {curve {rowName=Tandem.Speech.Priority.OutfitReaction}}
     }
     {
      message=I'm Nisha. I'm new to The Underground. Which I imagine you'd guessed.
      contextRequirements [{requirement {class=FortControllerRequirement_HasCID, properties {allowedCIds [/BRCosmetics/Athena/Items/Cosmetics/Characters/Character_SugarBubble]}}}]
      priority {curve {rowName=Tandem.Speech.Priority.OutfitReaction}}
     }
    ]
   }
  }
 }
 ConversationLinkNode_2 {
  id=#5
  class=ConversationLinkNode
  properties {remoteEntryTag {tagName=Conversation.Entry.CeremonialGuardFencer.Dialogue}}
 }
 ConversationLinkNode_0 {id=#4, class=ConversationLinkNode, properties {remoteEntryTag {tagName=Conversation.Entry.BasicServices}}}
 ConversationLinkNode_1 {id=#3, class=ConversationLinkNode, properties {remoteEntryTag {tagName=Conversation.Entry.NPCLibrary.Quests}}}
]
